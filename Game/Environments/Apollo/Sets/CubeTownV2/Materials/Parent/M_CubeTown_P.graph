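# Parent subsurface material for Apollo's CubeTown V2 cubes, with switches for single-cube and deco variants.
# Masks, interior-cubemap variation and MPC-driven sphere glows light the cubes, with lake-line emissive, height-lerped dust and flipbook glyph layers blended on top.
# Optional runtime virtual texture height blending, skipped by default, kills glow near terrain.
# Quality and shading-path switches simplify low-end rendering, then time of day scales emissive before MF_CreativeMoveTool.
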Material /Game/Environments/Apollo/Sets/CubeTownV2/Materials/Parent/M_CubeTown_P
graph {
 properties {
  bTangentSpaceNormal=false
  bUseMaterialAttributes=true
  bUsedWithNanite=true
  physMaterial=/Game/Environments/Apollo/Sets/Cubetown/Materials/CubeTown_PhysMat
  pixelDepthOffsetMode=PDOM_Legacy
  shadingModel=MSM_Subsurface
  usedShadingModels=Subsurface
 }
 nodes [
  n1 = Constant()
  n2 = NamedRerouteUsage()
  n3 = NamedRerouteUsage()
  n4 = Multiply(A=TextureCoordinate(), B=ScalarParameter() {defaultValue=1, parameterName=CubesUVScale})
  n5 = TextureSampleParameter2D(UVs=Multiply(A=@n4) {constB=0.5}) {
   group=Textures
   parameterName=CubesMask
   samplerSource=SSM_Wrap_WorldGroupSettings
   samplerType=SAMPLERTYPE_Masks
   texture=&/CubeTown_1_H
  }
  n6 = Constant() {r=1}
  n7 = Multiply(
   A=TextureCoordinate()
   B=Divide(
    A=ScalarParameter() {defaultValue=4, desc=Dived from tiled texture, group=Single, parameterName=SingeScale}
    B=ScalarParameter() {defaultValue=16, desc=Dived from tiled texture, group=Single, parameterName=SingeScaleDepthDevider}
   ) {constB=16}
  )
  n8 = TextureCoordinate()
  n9 = Add(
   A=MaterialFunctionCall!(UVs=StaticSwitchParameter(True=@n7, False=@n8) {group=Single, parameterName=SingeCubeTemp}, Tiling, Randomize Rotation) {materialFunction=/Engine/Functions/Engine_MaterialFunctions01/Cubemaps/InteriorCubemap}.UVW
   B=Constant3Vector() {constant=<rgba 0.9 0 512 0>}
  )
  n10 = Saturate(None=Divide(A=ComponentMask(None=@n9) {r=true}, B=Constant() {r=15}))
  n11 = LinearInterpolate(A=Constant(), B=Constant() {r=5}, Alpha=@n10)
  n12 = NamedRerouteUsage()
  n13 = TextureSampleParameter2D(UVs=@n4) {
   group=Textures
   parameterName=CubesMask
   samplerSource=SSM_Wrap_WorldGroupSettings
   samplerType=SAMPLERTYPE_Masks
   texture=&CubeTown_1_H
  }
  n14 = NamedRerouteDeclaration(None=Multiply(A=LinearInterpolate(A=@n6, B=@n11, Alpha=@n12), B=@n13.G)) {name=CubesMask_2, nodeColor=<rgba 1 0.9609 0 1>}
  n15 = LinearInterpolate(A=Constant() {desc=.35, r=0.25}, B=Constant() {desc=1, r=1}, Alpha=@n5.R)
  n16 = Multiply(A=@n15, B=@n14)
  n17 = TextureSampleParameter2D(UVs=TextureCoordinate() {coordinateIndex=1}) {
   group=Textures
   parameterName=GlowMask
   samplerSource=SSM_Wrap_WorldGroupSettings
   samplerType=SAMPLERTYPE_Masks
   texture=/Game/Environments/Apollo/Sets/Cubetown/Textures/Cube_1_M
  }
  n18 = VectorParameter() {parameterName=Color, defaultValue=<rgba 0.1863 0.0176 0.2427 1>}
  n19 = Subtract(A=Constant() {r=100}, B=ComponentMask(None=WorldPosition().XYZ) {b=true})
  n20 = Constant() {r=5000}
  n21 = NamedRerouteDeclaration(None=Saturate(None=Divide(A=@n19, B=@n20))) {name=Grad, nodeColor=<rgba 0 0.4141 1 1>}
  n22 = VectorParameter() {group=Glow, parameterName=OuterGlow, defaultValue=<rgba 0.5 0.1457 0.0975 0>}
  n23 = NamedRerouteUsage()
  n24 = WorldPosition()
  n25 = ComponentMask(None=CollectionParameter() {collection=/Game/Environments/Apollo/Sets/CubeTownV2/BP/MPC_CubetownGlow, parameterName=GlowLocation}) {b=true, g=true, r=true}
  n26 = ScalarParameter() {defaultValue=5000, group=Glow, parameterName=Outer Radius Glow}
  n27 = CollectionParameter() {collection=&MPC_CubetownGlow, parameterName=GlowRadiusScale}
  n28 = NamedRerouteDeclaration(None=SphereMask(A=@n24.XYZ, B=@n25, Radius=Multiply(A=@n26, B=@n27), Hardness=Constant())) {name=SphereMask, nodeColor=<rgba 0 0.6484 1 1>}
  n29 = MaterialFunctionCall(Transition Phase=@n28, Height Texture=@n23, Contrast=Constant() {r=1}) {materialFunction=&HeightLerp}
  n30 = Constant() {r=2.5}
  n31 = Constant() {r=2}
  n32 = Constant() {r=0.2}
  n33 = ComponentMask(None=VertexNormalWS()) {b=true}
  n34 = Add(
   A=Add(A=@n28, B=Multiply(A=@n29.Alpha, B=@n30))
   B=NamedRerouteDeclaration(None=LinearInterpolate(A=@n31, B=@n32, Alpha=@n33)) {name=UPSides, nodeColor=<rgba 1 0 0.5625 1>}
  )
  n35 = Multiply(A=@n18.A, B=Add(A=Multiply(A=@n21, B=@n22.RGB), B=Multiply(A=@n22.RGB, B=@n34)))
  n36 = Multiply(A=@n26, B=Multiply(A=@n27) {constB=0.5})
  n37 = Constant() {r=-2}
  n38 = CollectionParameter() {collection=&MPC_CubetownGlow, parameterName=GlowScale}
  n39 = StaticSwitchParameter(True=Constant() {r=0.5}, False=@n23) {group=Single, parameterName=SingeCubeTemp}
  n40 = Multiply(
   A=Multiply(A=SphereMask(A=@n24.XYZ, B=@n25, Radius=@n36, Hardness=@n37), B=Multiply(A=@n38, B=@n39))
   B=VectorParameter() {group=Center, parameterName=CenterGlow, defaultValue=<rgba 1 0.0983 0 0>}.RGB
  )
  n41 = Multiply(
   A=LinearInterpolate(A=Constant() {r=0.5}, B=Constant() {r=1.5}, Alpha=@n17.B)
   B=Add(A=Multiply(A=@n18.RGB, B=@n35), B=NamedRerouteDeclaration(None=@n40) {name=CentralGlow, nodeColor=<rgba 1 0 0.0937 1>})
  )
  n42 = Multiply(A=TextureCoordinate(), B=Constant() {r=0.125})
  n43 = TextureSampleParameter2D(UVs=BumpOffset(Coordinate=@n42, Height=Constant(), HeightRatioInput=Constant() {r=0.05})) {
   parameterName=Dirt
   samplerSource=SSM_Wrap_WorldGroupSettings
   samplerType=SAMPLERTYPE_Masks
   texture=&/T_CubeTownDirt_M
  }
  n44 = Constant() {r=2}
  n45 = TextureSampleParameter2D(UVs=BumpOffset(Coordinate=Multiply(A=@n42, B=@n44), Height=Constant() {r=0.25}, HeightRatioInput=Constant() {r=0.05})) {
   parameterName=Dirt
   samplerSource=SSM_Wrap_WorldGroupSettings
   samplerType=SAMPLERTYPE_Masks
   texture=&T_CubeTownDirt_M
  }
  n46 = LinearInterpolate(A=Constant() {r=0.5}, B=Constant() {r=1.25}, Alpha=Min(A=@n43.A, B=@n45.A))
  n47 = StaticSwitchParameter(True=LinearInterpolate(A=Constant() {r=1}, B=@n46, Alpha=NamedRerouteUsage()), False=@n46) {parameterName=Deco}
  n48 = TextureSampleParameter2D() {
   group=Textures
   parameterName=CubesN
   samplerSource=SSM_Wrap_WorldGroupSettings
   samplerType=SAMPLERTYPE_Normal
   texture=&/CubeTown_1_N
  }
  n49 = SetMaterialAttributes!(
   MaterialAttributes
   Base Color=VectorParameter() {parameterName=Color, defaultValue=<rgba 0.1863 0.0176 0.2427 1>}.RGB
   Metallic=ScalarParameter() {defaultValue=0.5, parameterName=Met}
   Specular=ScalarParameter() {defaultValue=0.25, parameterName=Spec}
   Roughness=LinearInterpolate(A=Constant(), B=Constant() {r=0.7}, Alpha=@n5.R)
   Emissive Color=Multiply(A=Multiply(A=@n16, B=@n41), B=NamedRerouteDeclaration(None=@n47) {name=Flakes, nodeColor=<rgba 0 1 0.0156 1>})
   Normal=NamedRerouteDeclaration(None=Transform(None=@n48.RGB)) {name=NormalMap, nodeColor=<rgba 0 0.25 1 1>}
   Opacity=Constant() {r=0.05}
   Subsurface Color=Multiply(A=@n14, B=VectorParameter() {parameterName=SSSColor, defaultValue=<rgba 0.1024 0 0.002 0>}.RGB)
  ) {attributeSetTypes [
    #1
    #2
    #3
    #4
    #5
    #6
    #7
    #8
   ]}
  n50 = GetMaterialAttributes(None=NamedRerouteDeclaration(None=@n49) {name=Cut For Low Spec, nodeColor=<rgba 0 1 0.4844 1>}) {attributeGetTypes [#5]}
  n51 = ComponentMask(None=WorldPosition().XYZ) {b=true}
  n52 = Constant() {r=2659.647}
  n53 = Saturate(None=Divide(A=Add(A=@n51, B=@n52), B=Constant() {r=-909.9397}))
  n54 = Multiply(A=NamedRerouteUsage(), B=@n53)
  n55 = Add(A=@n51, B=Constant() {r=3075})
  n56 = Constant() {r=-50}
  n57 = Add(A=Saturate(None=Divide(A=@n55, B=@n56)), B=Multiply(A=@n53, B=Constant() {r=0.05}))
  n58 = NamedRerouteDeclaration(None=Multiply(A=VectorParameter() {parameterName=LakeLine, defaultValue=<rgba 15 0 2.304 0>}.RGB, B=Add(A=@n54, B=@n57))) {name=LakeLine, nodeColor=<rgba 1 0.5391 0 1>}
  n59 = GetMaterialAttributes(None=SetMaterialAttributes!(MaterialAttributes=@n50.MaterialAttributes, Emissive Color=Add(A=@n50.Emissive Color, B=@n58)) {attributeSetTypes [#5]}) {attributeGetTypes [#4]}
  n60 = TextureSampleParameter2D() {
   parameterName=Dirt
   samplerSource=SSM_Wrap_WorldGroupSettings
   samplerType=SAMPLERTYPE_Masks
   texture=&T_CubeTownDirt_M
  }
  n61 = LinearInterpolate(A=Constant() {r=0.1}, B=Constant() {r=0.6762}, Alpha=@n60.G)
  n62 = ComponentMask(None=VertexNormalWS()) {b=true}
  n63 = SetMaterialAttributes!(
   MaterialAttributes=@n59.MaterialAttributes
   Roughness=LinearInterpolate(A=Multiply(A=@n59.Roughness, B=@n61), B=@n59.Roughness, Alpha=Abs(None=@n62))
  ) {attributeSetTypes [#4]}
  n64 = GetMaterialAttributes(None=@n63) {attributeGetTypes [#5]}
  n65 = SetMaterialAttributes!(
   MaterialAttributes=@n64.MaterialAttributes
   Roughness=Constant() {r=1}
   Metallic=Constant()
   Base Color=VectorParameter() {parameterName=DustColor, defaultValue=<rgba 0.4601 0.0589 0.0359 0>}.RGB
   Specular=Constant()
   Emissive Color=Multiply(A=@n64.Emissive Color, B=Constant() {r=0.5})
  ) {attributeSetTypes [
    #4
    #2
    #1
    #3
    #5
   ]}
  n66 = Constant() {r=0.5}
  n67 = Multiply(A=TextureCoordinate(), B=Constant() {r=0.25})
  n68 = Multiply(
   A=TextureSample(UVs=Multiply(A=@n66, B=@n67)) {texture=/Game/Weapons/FORT_Melee/Effects/LaserSwordFounders/T_LaserSwordGlow}.R
   B=TextureSample(UVs=@n67) {texture=&T_LaserSwordGlow}.R
  )
  n69 = OneMinus(None=@n68)
  n70 = Multiply(A=ScalarParameter() {defaultValue=0.1, parameterName=UpDirt}, B=ComponentMask(None=VertexNormalWS()) {b=true})
  n71 = TextureSampleParameter2D(UVs=TextureCoordinate() {coordinateIndex=1}) {group=Textures, parameterName=GlowMask, samplerType=SAMPLERTYPE_Masks, texture=&Cube_1_M}
  n72 = TextureSampleParameter2D(UVs=Multiply(A=Constant() {r=2.7471}, B=TextureCoordinate())) {
   parameterName=Dirt
   samplerSource=SSM_Wrap_WorldGroupSettings
   samplerType=SAMPLERTYPE_Masks
   texture=&T_CubeTownDirt_M
  }
  n73 = Constant() {r=0.2603}
  n74 = VertexNormalWS()
  n75 = MaterialFunctionCall(
   Transition Phase=Multiply(A=OneMinus(None=@n70), B=@n71.R)
   Height Texture=LinearInterpolate(A=@n72.R, B=Multiply(A=@n72.B, B=@n73), Alpha=ComponentMask(None=@n74) {b=true})
  ) {materialFunction=&HeightLerp}
  n76 = BlendMaterialAttributes(A=@n65, B=@n63, Alpha=Saturate(None=Add(A=@n69, B=@n75.Alpha)))
  n77 = GetMaterialAttributes(None=@n76) {attributeGetTypes [#1]}
  n78 = SetMaterialAttributes!(
   MaterialAttributes=@n77.MaterialAttributes
   Base Color=Multiply(A=@n77.Base Color, B=Constant() {r=0.5})
   Specular=Constant() {r=0.05}
   Roughness=Constant()
   Normal=Constant3Vector() {constant=<rgba 0 0 1 0>}
  ) {attributeSetTypes [
    #1
    #3
    #4
    #6
   ]}
  n79 = Constant() {r=-0.5}
  n80 = Constant() {r=2}
  n81 = TextureSampleParameter2D() {
   group=Textures
   parameterName=CubesMask
   samplerSource=SSM_Wrap_WorldGroupSettings
   samplerType=SAMPLERTYPE_Masks
   texture=&CubeTown_1_H
  }
  n82 = Constant() {r=1.65}
  n83 = LinearInterpolate(
   B=MaterialFunctionCall(Transition Phase=LinearInterpolate(A=@n79, B=@n80, Alpha=@n68), Height Texture=Multiply(A=@n81.G, B=@n82)) {materialFunction=&HeightLerp}.Lerp Alpha No Contrast
   Alpha=ComponentMask(None=VertexNormalWS()) {b=true}
  ) {constA=1}
  n84 = ShadingPathSwitch(Default=QualitySwitch(Default=BlendMaterialAttributes(A=@n78, B=@n76, Alpha=@n83), Low=@n3), Mobile=@n3)
  n85 = GetMaterialAttributes(None=BlendMaterialAttributes(A=StaticSwitchParameter(True=@n2, False=@n84) {group=Single, parameterName=SingeCubeTemp}, B=NamedRerouteUsage(), Alpha=NamedRerouteUsage()) {vertexAttributeBlendType=UseA}) {attributeGetTypes [#5]}
  n86 = Fresnel(ExponentIn=Constant() {r=2.5}, BaseReflectFractionIn=Constant())
  n87 = SetMaterialAttributes!(MaterialAttributes=@n85.MaterialAttributes, Emissive Color=Multiply(A=@n85.Emissive Color, B=OneMinus(None=@n86))) {attributeSetTypes [#5]}
  n88 = SetMaterialAttributes!(MaterialAttributes=@n87, Emissive Color=@n1, Subsurface Color=@n1) {attributeSetTypes [#5, #8]}
  n89 = RuntimeVirtualTextureSampleParameter() {materialType=WorldHeight, parameterName=RVTHeight, virtualTexture=/Game/Athena/Apollo/Maps/Landscape/VT/RVT_Apollo_Landscape_Height}
  n90 = Subtract(A=ComponentMask(None=WorldPosition().XYZ) {b=true}, B=@n89.WorldHeight)
  n91 = ScalarParameter() {defaultValue=256, group=RVT, parameterName=Blend RVT}
  n92 = TextureCoordinate()
  n93 = ScalarParameter() {defaultValue=4, desc=Dived from tiled texture, group=Single, parameterName=SingeScale}
  n94 = TextureSampleParameter2D(UVs=StaticSwitchParameter(True=Divide(A=@n92, B=@n93) {constB=4}, False=@n92) {group=Single, parameterName=SingeCubeTemp}) {
   group=Textures
   parameterName=CubesMask
   samplerSource=SSM_Wrap_WorldGroupSettings
   samplerType=SAMPLERTYPE_Masks
   texture=&CubeTown_1_H
  }
  n95 = MaterialFunctionCall(Transition Phase=Saturate(None=Divide(A=@n90, B=@n91)), Height Texture=Max(A=@n94.B, B=Constant() {r=0.25})) {materialFunction=&HeightLerp}
  n96 = VirtualTextureFeatureSwitch!(
   No=@n87
   Yes=MaterialProxyReplace(Realtime=BlendMaterialAttributes(A=@n88, B=@n87, Alpha=@n95.Alpha) {vertexAttributeBlendType=UseA}, MaterialProxy=@n87)
  )
  n97 = ShadingPathSwitch(Default=@n96, Deferred=@n96, Forward=@n87, Mobile=@n87)
  n98 = StaticSwitchParameter(True=@n87, False=QualitySwitch(Default=@n87, Low=@n87, High=@n97, Medium=@n97, Epic=@n97)) {
   defaultValue=true
   desc="If for some reason you want to use this material but hate to use RVT ( or you'r Bryce L) Will automatically convert details to use up vector instead distance from the ground "
   group=RVT
   parameterName=Skip RVT
   sortPriority=-4
  }
  n99 = GetMaterialAttributes(None=@n98) {attributeGetTypes [#5]}
  n100 = CollectionParameter() {
   collection=&FortniteMaterialParameters
   parameterName=FortniteActiveTimeOfDay
  }
  n101 = VectorParameter() {parameterName=ToD, defaultValue=<rgba 0.25 0.75 0.25 1>}
  n102 = DotProduct(A=@n100, B=AppendVector(A=@n101.RGB, B=@n101.RGB))
  MaterialFunctionCall(MA In=SetMaterialAttributes!(MaterialAttributes=@n99.MaterialAttributes, Emissive Color=Multiply(A=@n99.Emissive Color, B=@n102)) {attributeSetTypes [#5]}) {materialFunction=&MF_CreativeMoveTool}
  n103 = ScalarParameter() {defaultValue=0.9623, parameterName=StarsTileMod}
  n104 = TextureCoordinate() {uTiling=0.25, vTiling=0.25}
  n105 = Multiply(A=Constant() {r=14}, B=@n104)
  n106 = Constant() {r=-2}
  n107 = Constant() {r=0.02}
  n108 = TextureSample(UVs=Multiply(A=@n103, B=BumpOffset(Coordinate=@n105, Height=@n106, HeightRatioInput=@n107))) {samplerSource=SSM_Wrap_WorldGroupSettings, samplerType=SAMPLERTYPE_Masks, texture=/Game/Global/Textures/Default/Noise/T_CubeNoise}
  n109 = Multiply(A=@n104, B=Constant() {r=28})
  n110 = Constant() {r=0.2}
  n111 = TextureSample(UVs=Multiply(A=@n103, B=BumpOffset(Coordinate=@n109, Height=@n106, HeightRatioInput=@n110))) {samplerSource=SSM_Wrap_WorldGroupSettings, samplerType=SAMPLERTYPE_Masks, texture=&T_CubeNoise}
  n112 = Power(Base=Multiply(A=@n108.B, B=@n111.B), Exp=Constant() {r=7})
  n113 = Constant() {r=60}
  n114 = TextureSampleParameter2D() {
   group=Textures
   parameterName=CubesMask
   samplerSource=SSM_Wrap_WorldGroupSettings
   samplerType=SAMPLERTYPE_Masks
   texture=&CubeTown_1_H
  }
  NamedRerouteDeclaration(None=Multiply(A=Multiply(A=@n112, B=@n113), B=@n114.G)) {name=Sparkles, nodeColor=<rgba 0 0.5078 1 1>}
  NamedRerouteDeclaration(None=@n5.R) {name=CubesMask_3, nodeColor=<rgba 1 0.9609 0 1>}
  n115 = VectorParameter() {parameterName=Color, defaultValue=<rgba 0.1863 0.0176 0.2427 1>}
  n116 = ScalarParameter() {defaultValue=0.119, parameterName=Rough2}
  n117 = ScalarParameter() {defaultValue=0.2111, parameterName=Rough}
  n118 = ObjectRadius()
  n119 = Constant() {r=126.9318}
  n120 = TextureSampleParameter2D(UVs=Multiply(A=Divide(A=@n118, B=@n119), B=TextureCoordinate())) {
   parameterName=Dirt
   samplerSource=SSM_Wrap_WorldGroupSettings
   samplerType=SAMPLERTYPE_Masks
   texture=&T_CubeTownDirt_M
  }
  n121 = VertexNormalWS()
  n122 = LinearInterpolate(A=@n120.G, B=Add(A=OneMinus(None=@n120.B), B=Constant() {r=0.219}), Alpha=Abs(None=ComponentMask(None=@n121) {b=true}))
  n123 = Divide(A=TextureCoordinate(), B=ScalarParameter() {defaultValue=4, desc=Dived from tiled texture, group=Single, parameterName=SingeScale}) {constB=4}
  n124 = TextureSampleParameter2D(UVs=@n123) {
   group=Textures
   parameterName=CubesMask
   samplerSource=SSM_Wrap_WorldGroupSettings
   samplerType=SAMPLERTYPE_Masks
   texture=&CubeTown_1_H
  }
  n125 = NamedRerouteUsage()
  n126 = VectorParameter() {group=Glow, parameterName=OuterGlow, defaultValue=<rgba 0.5 0.1457 0.0975 0>}
  n127 = Add(A=NamedRerouteUsage(), B=Constant() {r=1})
  n128 = Multiply(A=@n115.RGB, B=Add(A=Multiply(A=@n125, B=@n126.RGB), B=Multiply(A=@n126.RGB, B=@n127)))
  n129 = NamedRerouteUsage()
  n130 = Multiply(A=Multiply(A=@n124.G, B=Add(A=@n128, B=@n129)), B=NamedRerouteUsage())
  n131 = MaterialFunctionCall(UVs=Divide(A=TextureCoordinate(), B=ScalarParameter() {defaultValue=4, desc=Dived from tiled texture, group=Single, parameterName=SingeScale}) {constB=4}) {materialFunction=&InteriorCubemap}
  n132 = Constant3Vector() {constant=<rgba 0.9 0 512 0>}
  n133 = Divide(A=ComponentMask(None=Add(A=@n131.UVW, B=@n132)) {r=true}, B=Constant() {r=15})
  n134 = LinearInterpolate(
   A=Constant() {r=0.854}
   B=Constant() {r=5}
   Alpha=NamedRerouteDeclaration(None=Saturate(None=@n133)) {name=SingleCubeMask, nodeColor=<rgba 0.1484 0 1 1>}
  )
  n135 = TextureSampleParameter2D(UVs=@n123) {
   group=Textures
   parameterName=CubesN
   samplerSource=SSM_Wrap_WorldGroupSettings
   samplerType=SAMPLERTYPE_Normal
   texture=&CubeTown_1_N
  }
  n136 = ScalarParameter() {group=Single, parameterName=FlattenNormal}
  n137 = SetMaterialAttributes!(
   MaterialAttributes
   Base Color=@n115.RGB
   Metallic=ScalarParameter() {defaultValue=0.5, parameterName=Met}
   Specular=ScalarParameter() {defaultValue=0.25, parameterName=Spec}
   Roughness=StaticSwitchParameter(True=@n116, False=LinearInterpolate(A=@n116, B=@n117, Alpha=@n122)) {parameterName=Deco}
   Emissive Color=Add(A=Multiply(A=@n130, B=@n134), B=NamedRerouteUsage())
   Normal=Transform(None=MaterialFunctionCall(Normal=@n135.RGB, Flatness=@n136) {materialFunction=&FlattenNormal}.Result)
   Opacity=Constant() {r=0.05}
   Subsurface Color=Multiply(A=VectorParameter() {parameterName=SSSColor, defaultValue=<rgba 0.1024 0 0.002 0>}.RGB, B=@n124.G)
  ) {attributeSetTypes [
    #1
    #2
    #3
    #4
    #5
    #6
    #7
    #8
   ]}
  NamedRerouteDeclaration(None=@n137) {name=SingeCube, nodeColor=<rgba 1 0.7969 0 1>}
  NamedRerouteDeclaration(None=LinearInterpolate(Alpha=@n10) {constB=4}) {name=TilesMask, nodeColor=<rgba 0 0.9766 1 1>}
  NamedRerouteDeclaration(None=@n15) {name=CubesMask, nodeColor=<rgba 1 0.9609 0 1>}
  n138 = VectorParameter() {group=Center, parameterName=GlyphGlow, defaultValue=<rgba 100 8.2472 0 0>}
  n139 = Saturate(None=DotProduct(A=VertexNormalWS(), B=CameraVectorWS()))
  n140 = TextureSample(UVs=Divide(A=TextureCoordinate(), B=ScalarParameter() {defaultValue=4, desc=Dived from tiled texture, group=Single, parameterName=SingeScale}) {constB=4}) {
   samplerSource=SSM_Wrap_WorldGroupSettings
   samplerType=SAMPLERTYPE_Masks
   texture=&CubeTown_1_H
  }
  n141 = Constant()
  n142 = SetMaterialAttributes!(
   MaterialAttributes
   Emissive Color=Multiply(A=Multiply(A=@n138.RGB, B=@n139), B=@n140.G)
   Specular=@n141
   Metallic=Constant() {r=1}
   Roughness=@n141
  ) {attributeSetTypes [
    #5
    #3
    #2
    #4
   ]}
  NamedRerouteDeclaration(None=@n142) {name=Glyph Mat, nodeColor=<rgba 1 0.0234 0 1>}
  n143 = NamedRerouteUsage()
  n144 = VertexNormalWS()
  n145 = Abs(None=ComponentMask(None=Transform(None=@n144) {transformSourceType=TRANSFORMSOURCE_World, transformType=TRANSFORM_Local}) {r=true})
  n146 = Constant() {r=0.15}
  n147 = Constant() {r=2}
  n148 = TextureCoordinate()
  n149 = Multiply(A=@n148, B=ScalarParameter() {defaultValue=1, group=Single, parameterName=GlyphTileSingle})
  n150 = ScalarParameter() {defaultValue=-1, group=Single, parameterName=GlyphDepthSingle}
  n151 = Constant() {r=0.1643}
  n152 = ComponentMask(None=MaterialFunctionCall(
    "Animation  Phase (0-1)"=ScalarParameter() {defaultValue=0.25, group=Single, parameterName=GlypthFrame}
    Number of Rows=@n147
    Number of Columns=@n147
    Texture=TextureObject() {samplerType=SAMPLERTYPE_LinearColor, texture=/Game/Athena/Prototype/Blueprints/Cube/Materials/DFHeightMask}
    UVs=BumpOffset(Coordinate=@n149, Height=@n150, HeightRatioInput=@n151)
   ) {materialFunction=&FlipBook}.Result) {r=true}
  n153 = Saturate(None=LinearInterpolate(A=Constant() {r=-0.5}, B=Constant() {r=1}, Alpha=NamedRerouteUsage()))
  n154 = Multiply(A=@n148, B=Constant() {r=2})
  n155 = Multiply(A=ScalarParameter() {defaultValue=-1, group=Center, parameterName=GlypthDepth}, B=@n140.G)
  n156 = MaterialFunctionCall(
   Transition Phase=@n146
   Height Texture=TextureSample(UVs=BumpOffset(Coordinate=@n154, Height=@n155, HeightRatioInput=@n151)) {samplerType=SAMPLERTYPE_LinearColor, texture=&DFHeightMask}.R
  ) {materialFunction=&HeightLerp}
  n157 = Multiply(
   A=StaticSwitchParameter(
    True=MaterialFunctionCall(Transition Phase=@n146, Height Texture=@n152) {materialFunction=&HeightLerp}.Alpha
    False=Multiply(A=@n153, B=@n156.Lerp Alpha No Contrast)
   ) {group=Single, parameterName=SingeCubeTemp}
   B=Constant() {r=0.1}
  )
  n158 = StaticSwitchParameter(True=Multiply(A=NamedRerouteDeclaration(None=@n145) {name=DecoXMask, nodeColor=<rgba 0.7812 1 0 1>}, B=@n157), False=@n157) {parameterName=Deco}
  n159 = CollectionParameter() {collection=&MPC_CubetownGlow, parameterName=GlowRadiusScale}
  n160 = SphereMask(
   A=WorldPosition().XYZ
   B=ComponentMask(None=CollectionParameter() {collection=&MPC_CubetownGlow, parameterName=GlowLocation}) {b=true, g=true, r=true}
   Radius=Multiply(A=ScalarParameter() {defaultValue=3000, group=Center, parameterName=SphereRadius2}, B=Multiply(A=@n159) {constB=1.5})
   Hardness=ScalarParameter() {group=Center, parameterName=SphereHardness2}
  )
  NamedRerouteDeclaration(None=Multiply(A=Multiply(A=@n143, B=@n158), B=NamedRerouteDeclaration(None=@n160) {name=CentralGlow_2, nodeColor=<rgba 1 0 0.2344 1>})) {name=GlyphMask, nodeColor=<rgba 0 0.3672 1 1>}
 ]
}
