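# Parent material for the CubeTown pyramid: a metallic surface with an emissive glow that responds to Fortnite time of day.
# Sphere masks around an offset prism point blend outer and inner colors and metal response, shaped by a mask texture, Fresnel and ground-height fades.
# Layered material attributes add two bump-offset dirt samples for fake depth, world-aligned dirt roughness on upward faces, and a height-mapped tip highlight.
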Material /Game/Environments/Apollo/Sets/CubeTownV2/Materials/Parent/M_Pyramid
graph {
 properties {bUseMaterialAttributes=true, pixelDepthOffsetMode=PDOM_Legacy, usedShadingModels=Default Lit}
 nodes [
  n1 = OneMinus(None=Fresnel(ExponentIn=Constant() {r=1.5}, BaseReflectFractionIn=Constant()))
  Reroute(None=@n1)
  n2 = NamedRerouteDeclaration(None=@n1) {name=FresnelFade, nodeColor=<rgba 0 0.1562 1 1>}
  n3 = TextureSample() {samplerType=SAMPLERTYPE_Masks, texture=/Game/Environments/Apollo/Sets/Cubetown/Textures/T_Pyramid_M}
  n4 = VectorParameter() {parameterName=Color, channelNames {a=ScaleEmissive, b="", g="", r=""}, defaultValue=<rgba 0.09 0.0085 0.1172 1.5>}
  n5 = CollectionParameter() {
   collection=&FortniteMaterialParameters
   parameterName=FortniteActiveTimeOfDay
  }
  n6 = Multiply(A=@n4.ScaleEmissive, B=DotProduct(A=Constant4Vector() {constant=<rgba 0.4 0.15 0.4 1>}.RGBA, B=@n5))
  n7 = VectorParameter() {parameterName=Color Inside, channelNames {a=ScaleEmissive, b="", g="", r=""}, defaultValue=<rgba 0.5 0.0821 0 5>}
  n8 = Constant4Vector() {constant=<rgba 0.25 0 0.25 1>}
  n9 = Multiply(A=@n7, B=Multiply(A=@n7.ScaleEmissive, B=DotProduct(A=@n5, B=@n8.RGBA)))
  n10 = WorldPosition()
  n11 = Add(A=ObjectPositionWS(), B=VectorParameter() {parameterName=Prism Location Offset, defaultValue=<rgba 0 0 6170 0>}.RGB)
  n12 = SphereMask(A=@n10.XYZ, B=@n11, Radius=ScalarParameter() {defaultValue=7336.2368, parameterName=Radius inside}, Hardness=Constant())
  n13 = ScalarParameter() {defaultValue=16007.8594, parameterName=Radius Outside}
  n14 = Constant()
  n15 = Multiply(
   A=Max(A=@n3.B, B=ScalarParameter() {defaultValue=0.3, parameterName=EmissiveShadowMin})
   B=LinearInterpolate(A=Multiply(A=@n4, B=@n6), B=LinearInterpolate(A=@n7, B=@n9, Alpha=@n12), Alpha=SphereMask(A=@n10.XYZ, B=@n11, Radius=@n13, Hardness=@n14))
  )
  n16 = LinearInterpolate(
   A=MaterialFunctionCall(Base=@n15, Blend=@n3.G) {materialFunction=&Blend_Overlay}.Result
   B=@n15
   Alpha=ScalarParameter() {defaultValue=2, parameterName=CurveTo Glow}
  )
  n17 = ScalarParameter() {defaultValue=0.377, parameterName=R}
  n18 = ScalarParameter() {defaultValue=0.125, parameterName=R2}
  n19 = ComponentMask(None=Subtract(A=WorldPosition().XYZ, B=ActorPositionWS())) {b=true}
  n20 = ScalarParameter() {defaultValue=7000, parameterName=Fade Roughness}
  n21 = NamedRerouteDeclaration(None=Saturate(None=Divide(A=@n19, B=@n20))) {name=GroundFade, nodeColor=<rgba 1 0 0.0937 1>}
  n22 = SetMaterialAttributes!(
   MaterialAttributes
   Emissive Color=Multiply(A=Multiply(A=@n2, B=@n16), B=NamedRerouteUsage())
   Base Color=@n4
   Metallic=LinearInterpolate(A=ScalarParameter() {defaultValue=0.5, parameterName=Met}, B=ScalarParameter() {defaultValue=1, parameterName=Param_5}, Alpha=@n12)
   Specular=LinearInterpolate(A=ScalarParameter() {defaultValue=0.15, parameterName=Spec}, Alpha=@n12) {constB=0}
   Roughness=LinearInterpolate(A=Constant() {r=1}, B=LinearInterpolate(A=@n17, B=@n18, Alpha=@n21), Alpha=@n3.R)
  ) {attributeSetTypes [
    #1
    #2
    #3
    #4
    #5
   ]}
  n23 = GetMaterialAttributes(None=@n22) {attributeGetTypes [#1, #5]}
  n24 = BumpOffset(Coordinate=Multiply(A=TextureCoordinate(), B=Constant() {r=4}), Height=Multiply(A=@n3.B, B=Constant() {r=-1.7484}))
  n25 = TextureObject() {samplerType=SAMPLERTYPE_Masks, texture=&/T_CubeTownDirt_M}
  n26 = BumpOffset(Coordinate=@n24, Height=Constant() {r=-1.1943})
  n27 = LinearInterpolate(
   A=Max(
    A=TextureSample(UVs=@n24, Tex=@n25) {samplerType=SAMPLERTYPE_Masks, texture=&T_CubeTownDirt_M}.A
    B=TextureSample(UVs=@n26) {samplerType=SAMPLERTYPE_Masks, texture=&T_CubeTownDirt_M}.A
   )
   Alpha=OneMinus(None=NamedRerouteUsage())
  )
  n28 = Constant() {r=-0.1143}
  n29 = Constant() {r=0.7079}
  n30 = MaterialFunctionCall(TextureObject=@n25, TextureSize=Constant() {r=1582.3422}) {materialFunction=&WorldAlignedTexture}
  n31 = ComponentMask(None=@n30.Z Texture) {r=true}
  n32 = LinearInterpolate(
   A=@n23.Roughness
   B=LinearInterpolate(A=LinearInterpolate(A=@n28, B=@n29, Alpha=@n31), B=Constant(), Alpha=NamedRerouteUsage())
   Alpha=ComponentMask(None=@n30.XY Texture) {g=true}
  )
  n33 = Saturate(None=ComponentMask(None=VertexNormalWS()) {b=true})
  n34 = GetMaterialAttributes(None=SetMaterialAttributes!(
    MaterialAttributes=@n23.MaterialAttributes
    Emissive Color=Multiply(A=@n23.Emissive Color, B=@n27)
    Roughness=LinearInterpolate(A=@n23.Roughness, B=@n32, Alpha=@n33)
   ) {attributeSetTypes [#1, #5]}) {attributeGetTypes [#1, #5]}
  n35 = Multiply(A=VectorParameter() {group=Tip, parameterName=TipColor, defaultValue=<rgba 500 229.9131 0 0>}.RGB, B=NamedRerouteUsage())
  n36 = Add(
   A=Subtract(A=WorldPosition().XYZ, B=ActorPositionWS())
   B=OneMinus(None=ScalarParameter() {defaultValue=8083.4922, group=Tip, parameterName=TipHeigh})
  )
  n37 = TextureSample(UVs=Divide(A=ComponentMask(None=@n36) {b=true}, B=ScalarParameter() {defaultValue=845.2463, group=Tip, parameterName=Tip Size})) {
   samplerSource=SSM_Clamp_WorldGroupSettings
   samplerType=SAMPLERTYPE_Grayscale
   texture=&/PiraTip
  }
  SetMaterialAttributes(
   MaterialAttributes=@n34.MaterialAttributes
   Emissive Color=LinearInterpolate(A=@n34.Emissive Color, B=Multiply(A=@n34.Emissive Color, B=@n35), Alpha=@n37.R)
   Roughness=@n34.MaterialAttributes
  ) {attributeSetTypes [#1, #5]}
 ]
}
